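# Masked material for the LightWheel vehicle cosmetic, adding an animated emissive glow to part of the MF_WheelParent_Simple base.
# A panning shockwave texture on rotated UVs, tinted and boosted by EmissiveColor, replaces base emissive where rim-mask blue and inverted vertex red overlap.
# The LOD2345 static switch confines the pulse to high-V UVs and fills the rest with flat color.
# The result then passes through the vehicle damage and de-rez transition functions.
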Material /VehicleCosmetics/Wheels/LightWheel/Materials/M_Wheel_LightWheel
graph {
 properties {
  bAutomaticallySetUsageInEditor=false
  bUseMaterialAttributes=true
  bUsedWithMeshDeformer=true
  bUsedWithSkeletalMesh=true
  blendMode=BLEND_Masked
  pixelDepthOffsetMode=PDOM_Legacy
  usedShadingModels=Default Lit
 }
 nodes [
  n1 = MaterialFunctionCall() {materialFunction=&MF_WheelParent_Simple}
  n2 = GetMaterialAttributes(None=@n1.MaterialAttributes) {attributeGetTypes [#1]}
  n3 = VectorParameter() {group=5_Special Effects, parameterName=EmissiveColor, channelNames {a=Emissive, b="", g="", r=""}, defaultValue=<rgba 0.2016 1 1 1>}
  n4 = Multiply(A=@n3.RGB, B=Multiply(A=@n3.Emissive) {constB=5})
  n5 = TextureCoordinate()
  n6 = Constant() {r=-0.25}
  n7 = Time()
  n8 = ScalarParameter() {defaultValue=1, group=5_Special Effects, parameterName=Speed}
  n9 = TextureSample(UVs=Panner(
    Coordinate=MaterialFunctionCall(UVs=@n5, "Rotation Angle (0-1)"=@n6) {materialFunction=&CustomRotator}.Rotated Values
    Time=Multiply(A=@n7, B=@n8)
   ) {speedY=-1}) {texture=/VehicleCosmetics/Wheels/_SharedTextures/FX/Curve_Shockwaves_Pack}
  n10 = Constant()
  n11 = If(A=ComponentMask(None=TextureCoordinate()) {g=true}, "A > B"=Constant() {r=1}, "A == B"=@n10, "A < B"=@n10) {constB=0.75}
  n12 = Multiply(A=@n4, B=Multiply(A=@n9.R, B=@n11))
  n13 = Multiply(A=@n3.RGB, B=OneMinus(None=@n11))
  n14 = Multiply(A=ComponentMask(None=@n1.RimRGBAMask) {b=true}, B=OneMinus(None=@n1.VertexColorR))
  n15 = StaticSwitchParameter(
   True=LinearInterpolate(A=@n2.Emissive Color, B=Add(A=@n12, B=@n13), Alpha=@n14)
   False=LinearInterpolate(A=@n2.Emissive Color, B=Multiply(A=@n4, B=@n9.R), Alpha=@n14)
  ) {group=5_Special Effects, parameterName=LOD2345}
  MaterialFunctionCall(In=MaterialFunctionCall(Material=SetMaterialAttributes(MaterialAttributes=@n2.MaterialAttributes, Emissive Color=@n15) {attributeSetTypes [#1]}) {materialFunction=&MF_BaseVehicle_Damage}.Result) {materialFunction=&MF_VehicleDeRez_Transition}
 ]
}
